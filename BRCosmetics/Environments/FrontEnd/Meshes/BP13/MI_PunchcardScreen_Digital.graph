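# Translucent lit material for a digital punchcard screen in the front-end environment, showing a punchcard icon over a blue panel.
# A radial UV gradient blends into the base color through a vertical edge mask, then the punchcard texture tints in a third color.
# Scrolling sine scanlines are added faintly on top.
# An inverted sphere mask between camera and object position drives icon and scanline visibility and shifts the edge threshold with viewing distance.
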Material /BRCosmetics/Environments/FrontEnd/Meshes/BP13/MI_PunchcardScreen_Digital
graph {
 properties {blendMode=BLEND_Translucent, pixelDepthOffsetMode=PDOM_Legacy, usedShadingModels=Default Lit}
 nodes [
  n1 = VectorParameter() {parameterName=RGB1, defaultValue=<rgba 0.0021 0.0056 0.0331 1>}
  n2 = MaterialFunctionCall(Vector 2=Add(A=TextureCoordinate()) {constB=-0.5}) {materialFunction=/Engine/Functions/Engine_MaterialFunctions02/Utility/VectorLength}
  n3 = LinearInterpolate(
   A=@n1.RGB
   B=VectorParameter() {parameterName=RGB2, defaultValue=<rgba 0.0003 0.0331 0.1329 1>}.RGB
   Alpha=Saturate(None=OneMinus(None=@n2.V2 Length))
  )
  n4 = TextureCoordinate()
  n5 = Multiply(A=Abs(None=Add(A=@n4) {constB=-0.5})) {constB=2}
  n6 = CameraPositionWS()
  n7 = ObjectPositionWS()
  n8 = ScalarParameter() {defaultValue=0.5, parameterName=FadeHardness}
  n9 = Saturate(None=OneMinus(None=SphereMask(A=@n6, B=@n7, Hardness=@n8)))
  n10 = Multiply(A=Subtract(A=Saturate(None=@n5), B=LinearInterpolate(Alpha=@n9) {constA=1, constB=0.75}) {constB=0.7}) {constB=128}
  n11 = Multiply(A=Saturate(None=ComponentMask(None=@n10) {g=true})) {constB=0.45}
  n12 = TextureCoordinate()
  n13 = VectorParameter() {parameterName=Scale-Offset, defaultValue=<rgba 3 1.8 -0.9 -0.3643>}
  n14 = ComponentMask(None=@n13.RGB) {g=true, r=true}
  n15 = TextureSample(UVs=Add(A=Multiply(A=@n12, B=@n14), B=AppendVector(A=@n13.B, B=@n13.B))) {samplerSource=SSM_Clamp_WorldGroupSettings, texture=/Game/Athena/UI/Challenges/Art/T_Ui_Punchcard_256}
  n16 = TextureCoordinate()
  n17 = Multiply(A=Time(), B=ScalarParameter() {defaultValue=0.02, parameterName=TimeFactr})
  n18 = Sine(None=Multiply(A=Add(A=@n16, B=@n17), B=ScalarParameter() {defaultValue=40, parameterName=TileScanline}))
  n19 = ScalarParameter() {defaultValue=0.75, parameterName=ScanlineEdge}
  n20 = ComponentMask(None=Multiply(A=Subtract(A=@n18, B=@n19) {constB=0.7}) {constB=64}) {g=true}
  n21 = LinearInterpolate(Alpha=@n9) {constB=0.001}
  Add(
   A=LinearInterpolate(
    A=LinearInterpolate(A=@n1.RGB, B=@n3, Alpha=@n11) {constA=0.75}
    B=VectorParameter() {parameterName=RGB3, defaultValue=<rgba 0.0003 0.162 0.4179 1>}.RGB
    Alpha=Multiply(A=@n15.R, B=@n9)
   )
   B=Saturate(None=Multiply(A=@n20, B=@n21))
  )
  LinearInterpolate(Alpha=@n9) {constA=0.8, constB=0.6}
  ScalarParameter() {defaultValue=0.6}
 ]
}
